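# Physics asset for the female medium Commando Bullseye helmet cosmetic, setting up its dangling left and right straps for simulation.
# Each strap is a three-capsule bone chain constrained to the head, with narrow swing on one axis and wide swing and twist on the others.
# The head capsule anchors both chains, while an unconstrained left clavicle capsule, with no right counterpart, appears to serve as a collider.
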PhysicsAsset /BRCosmetics/Accessories/Hats/F_MED_Commando_Bullesye/Meshes/F_MED_Commando_Bullseye_Helmet_Physics
bodies {
 clavicle_l [{
   shapeType=Capsule
   length=5.3211
   radius=2.9281
   shapeName=clavicle_l_capsule
   center=<3.3921 0.2952 7.9833>
   rotation=<rot -0.0345 5.5081 -89.6422>
  }]
 dyn_strap_01_l [{shapeType=Capsule, length=3.005, radius=1.005}]
 dyn_strap_01_r [{shapeType=Capsule, length=4.25, radius=1}]
 dyn_strap_02_l [{shapeType=Capsule, length=2.005, radius=1.255, center=<0 0 0.5153>}]
 dyn_strap_02_r [{shapeType=Capsule, length=3.505, radius=1.005, center=<0 0 -0.6089>, rotation=<rot 0 -90 0>}]
 dyn_strap_03_l [{shapeType=Capsule, length=2.3303, radius=0.8749, center=<0 0 -0.4358>}]
 dyn_strap_03_r [{shapeType=Capsule, length=3.627, radius=0.8795}]
 head [{shapeType=Capsule, length=10.0257, radius=7.2046, rotation=<rot 90 0 0>}]
}
constraints [
 {bone1Name=dyn_strap_01_r, bone2Name=head, swing1LimitDegrees=7, swing2LimitDegrees=70, twistLimitDegrees=70}
 {bone1Name=dyn_strap_02_r, bone2Name=dyn_strap_01_r, swing1LimitDegrees=7, swing2LimitDegrees=70, twistLimitDegrees=70}
 {bone1Name=dyn_strap_03_r, bone2Name=dyn_strap_02_r, swing1LimitDegrees=7, swing2LimitDegrees=70, twistLimitDegrees=70}
 {bone1Name=dyn_strap_01_l, bone2Name=head, swing1LimitDegrees=7, swing2LimitDegrees=70, twistLimitDegrees=70}
 {bone1Name=dyn_strap_02_l, bone2Name=dyn_strap_01_l, swing1LimitDegrees=7, swing2LimitDegrees=70, twistLimitDegrees=70}
 {bone1Name=dyn_strap_03_l, bone2Name=dyn_strap_02_l, swing1LimitDegrees=7, swing2LimitDegrees=70, twistLimitDegrees=70}
]
